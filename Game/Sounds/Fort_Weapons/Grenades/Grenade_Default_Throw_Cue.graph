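# Sound cue for the default grenade throw, reusing the Molotov throw wave from the FireJar grenade folder.
# A single wave player feeds the output at reduced volume, with no randomization or mixing.
# It uses gadget-placement attenuation and the gadgets and items sound class.
# A concurrency override prevents new instances from starting once the limit is reached.
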SoundCue /Game/Sounds/Fort_Weapons/Grenades/Grenade_Default_Throw_Cue
root=WavePlayer_0
nodes [{id=WavePlayer_0, properties {soundWaveAssetPtr=/Game/Sounds/Fort_Weapons/Grenades/FireJar/Molotov_Throw}}]
properties {
 attenuationSettings=/Game/Sounds/Attenuation/Fort_GadgetPlace_Att
 bOverrideConcurrency=true
 maxDistance=2000
 soundClassObject=&Fort_Gadgets_Items
 volumeMultiplier=0.2
 concurrencyOverrides {resolutionRule=PreventNew}
}
